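# Sound cue for the stink pod plant's explosion, mixing close and distant layers under the medium explosives sound class.
# Each layer has a modulator for variation and its own attenuation asset.
# A quality-level switch in each layer picks from eight random waves in its first two slots, or a three-wave subset excluded from branch culling in the last two.
# It references the rifle bullet impact concurrency set.
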SoundCue /ChronoEnviroItems/LaunchablePlants/Gameplay/Assets/Sounds/Stink_Plant/Chrono_PodPlant_Stink_Explode_Cue
root=Mixer_1
nodes [
 {
  id=Attenuation_1
  children [QualityLevel_0]
  properties {attenuationSettings=/ChronoEnviroItems/LaunchablePlants/Gameplay/Assets/Sounds/Launchable_Plants_Explode_Att}
 }
 {
  id=Attenuation_2
  children [QualityLevel_1]
  properties {attenuationSettings=/ChronoEnviroItems/LaunchablePlants/Gameplay/Assets/Sounds/Launchable_Plants_Explode_Distant_Att}
 }
 {id=Mixer_1, children [Modulator_0,Modulator_1], properties {inputVolume [1,1]}}
 {id=Modulator_0, children [Attenuation_1]}
 {id=Modulator_1, children [Attenuation_2]}
 {id=QualityLevel_0, children [Random_0,Random_0,Random_2,Random_2]}
 {id=QualityLevel_1, children [Random_1,Random_1,Random_3,Random_3]}
 {
  id=Random_0
  children [WavePlayer_2,WavePlayer_3,WavePlayer_4,WavePlayer_5,WavePlayer_6,WavePlayer_7,WavePlayer_8,WavePlayer_9]
  properties {weights [1,1,1,1,1,1,1,1]}
 }
 {
  id=Random_1
  children [WavePlayer_10, WavePlayer_11, WavePlayer_12, WavePlayer_13, WavePlayer_14, WavePlayer_15, WavePlayer_16, WavePlayer_17]
  properties {weights [1,1,1,1,1,1,1,1]}
 }
 {id=Random_2, children [WavePlayer_7,WavePlayer_8,WavePlayer_9], properties {bShouldExcludeFromBranchCulling=true, weights [1,1,1]}}
 {id=Random_3, children [WavePlayer_15, WavePlayer_16, WavePlayer_17], properties {bShouldExcludeFromBranchCulling=true, weights [1,1,1]}}
 {
  id=WavePlayer_10
  properties {soundWaveAssetPtr=&/PodPlant_PodExplode_Stink_Distant_01}
 }
 {
  id=WavePlayer_11
  properties {soundWaveAssetPtr=&/PodPlant_PodExplode_Stink_Distant_02}
 }
 {
  id=WavePlayer_12
  properties {soundWaveAssetPtr=&/PodPlant_PodExplode_Stink_Distant_03}
 }
 {
  id=WavePlayer_13
  properties {soundWaveAssetPtr=&/PodPlant_PodExplode_Stink_Distant_04}
 }
 {
  id=WavePlayer_14
  properties {soundWaveAssetPtr=&/PodPlant_PodExplode_Stink_Distant_05}
 }
 {
  id=WavePlayer_15
  properties {soundWaveAssetPtr=&/PodPlant_PodExplode_Stink_Distant_06}
 }
 {
  id=WavePlayer_16
  properties {soundWaveAssetPtr=&/PodPlant_PodExplode_Stink_Distant_07}
 }
 {
  id=WavePlayer_17
  properties {soundWaveAssetPtr=&/PodPlant_PodExplode_Stink_Distant_08}
 }
 {
  id=WavePlayer_2
  properties {soundWaveAssetPtr=&/PodPlant_PodExplode_Stink_Close_01}
 }
 {
  id=WavePlayer_3
  properties {soundWaveAssetPtr=&/PodPlant_PodExplode_Stink_Close_02}
 }
 {
  id=WavePlayer_4
  properties {soundWaveAssetPtr=&/PodPlant_PodExplode_Stink_Close_03}
 }
 {
  id=WavePlayer_5
  properties {soundWaveAssetPtr=&/PodPlant_PodExplode_Stink_Close_04}
 }
 {
  id=WavePlayer_6
  properties {soundWaveAssetPtr=&/PodPlant_PodExplode_Stink_Close_05}
 }
 {
  id=WavePlayer_7
  properties {soundWaveAssetPtr=&/PodPlant_PodExplode_Stink_Close_06}
 }
 {
  id=WavePlayer_8
  properties {soundWaveAssetPtr=&/PodPlant_PodExplode_Stink_Close_07}
 }
 {
  id=WavePlayer_9
  properties {soundWaveAssetPtr=&/PodPlant_PodExplode_Stink_Close_08}
 }
]
properties {
 maxDistance=12000
 soundClassObject=/Game/Sounds/FortSoundClassesAndModes/Fort_Explosives_Medium_PSM
 volumeMultiplier=1
 concurrencyOverrides {maxCount=4, resolutionRule=PreventNew}
 concurrencySet [&RifleBulletImpact_Conc]
}
