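# Sound cue for the launched Stink pod plant, mixing an in-air loop with launch and pod-bounce one-shots.
# The looping wave ramps up in volume through an Enveloper under in-air attenuation, while both one-shot layers share launched attenuation.
# Launch variants come from random pools switched by quality level, and the bounce layer borrows random Bomb plant waves.
# Modulators apply fixed pitch shifts, and a concurrency override caps simultaneous instances.
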SoundCue /ChronoEnviroItems/LaunchablePlants/Gameplay/Assets/Sounds/Stink_Plant/Chrono_PodPlant_Stink_InAirLoop_Cue
root=Mixer_0
nodes [
 {
  id=Attenuation_1
  children [Modulator_0]
  properties {attenuationSettings=/ChronoEnviroItems/LaunchablePlants/Gameplay/Assets/Sounds/Launchable_Plants_InAir_Loop_Att}
 }
 {
  id=Attenuation_2
  children [Modulator_3]
  properties {attenuationSettings=/ChronoEnviroItems/LaunchablePlants/Gameplay/Assets/Sounds/Launchable_Plants_Launched_Att}
 }
 {
  id=Attenuation_3
  children [Modulator_2]
  properties {attenuationSettings=&Launchable_Plants_Launched_Att}
 }
 {
  id=Enveloper_0
  children [WavePlayer_8]
  properties {pitchCurve {editorCurveData {keys [{time=0}]}}, volumeCurve {editorCurveData {keys [<at 0 0.15 RCIM_Cubic>,<at 1 1>]}}}
 }
 {id=Mixer_0, children [Attenuation_1, Attenuation_3, Attenuation_2], properties {inputVolume [0.75,0.2,1]}}
 {id=Modulator_0, children [Enveloper_0], properties {pitchMax=2, pitchMin=2, volumeMax=2, volumeMin=2}}
 {id=Modulator_2, children [QualityLevel_0], properties {pitchMax=2.5, pitchMin=2.5, volumeMax=1, volumeMin=1}}
 {id=Modulator_3, children [Random_2], properties {pitchMax=1, pitchMin=1}}
 {id=QualityLevel_0, children [Random_0,Random_0,Random_1,Random_1]}
 {id=Random_0, children [WavePlayer_2,WavePlayer_3,WavePlayer_4,WavePlayer_5], properties {weights [1,1,1,1]}}
 {id=Random_1, children [WavePlayer_3,WavePlayer_4,WavePlayer_5], properties {bShouldExcludeFromBranchCulling=true, weights [1,1,1]}}
 {id=Random_2, children [WavePlayer_9, WavePlayer_10, WavePlayer_11], properties {weights [1,1,1]}}
 {
  id=WavePlayer_10
  properties {soundWaveAssetPtr=/ChronoEnviroItems/LaunchablePlants/Gameplay/Assets/Sounds/Bomb_Plant/PodPlant_PodBounce_Bomb_01}
 }
 {
  id=WavePlayer_11
  properties {soundWaveAssetPtr=/ChronoEnviroItems/LaunchablePlants/Gameplay/Assets/Sounds/Bomb_Plant/PodPlant_PodBounce_Bomb_02}
 }
 {
  id=WavePlayer_2
  properties {soundWaveAssetPtr=&/PodPlant_Launch_Stink_01}
 }
 {
  id=WavePlayer_3
  properties {soundWaveAssetPtr=&/PodPlant_Launch_Stink_02}
 }
 {
  id=WavePlayer_4
  properties {soundWaveAssetPtr=&/PodPlant_Launch_Stink_03}
 }
 {
  id=WavePlayer_5
  properties {soundWaveAssetPtr=&/PodPlant_Launch_Stink_04}
 }
 {
  id=WavePlayer_8
  properties {bLooping=true, soundWaveAssetPtr=&/Chrono_PodPlant_Stink_InAir_Loop}
 }
 {
  id=WavePlayer_9
  properties {soundWaveAssetPtr=/ChronoEnviroItems/LaunchablePlants/Gameplay/Assets/Sounds/Bomb_Plant/PodPlant_PodBounce_Bomb_03}
 }
]
properties {
 bOverrideConcurrency=true
 maxDistance=6000
 pitchMultiplier=1.5
 soundClassObject=&Fort_WeaponRanged_Projectile_Effects
 volumeMultiplier=2
 concurrencyOverrides {maxCount=3}
 concurrencySet [null]
}
