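# Sound cue for the whip sound of the Chrono stink pod plant, mixing two randomized, pitched-down layers.
# The high-end layer switches by quality level, picking from five waves at the first two levels and a three-wave subset at the last two.
# The second layer picks from three base whip waves and is mixed much louder.
# It uses projectile-effects sound class and overridden in-air projectile attenuation, without reverb send.
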SoundCue /ChronoEnviroItems/LaunchablePlants/Gameplay/Assets/Sounds/Stink_Plant/Chrono_PodPlant_Stink_Whip_Cue
root=Mixer_0
nodes [
 {id=Mixer_0, children [Modulator_1,Modulator_0], properties {inputVolume [1,4]}}
 {id=Modulator_0, children [Random_1], properties {pitchMax=0.8, pitchMin=0.8}}
 {id=Modulator_1, children [QualityLevel_0], properties {pitchMax=0.7, pitchMin=0.7}}
 {id=QualityLevel_0, children [Random_3,Random_3,Random_0,Random_0]}
 {id=Random_0, children [WavePlayer_2,WavePlayer_5,WavePlayer_6], properties {bShouldExcludeFromBranchCulling=true, weights [1,1,1]}}
 {id=Random_1, children [WavePlayer_7, WavePlayer_8, WavePlayer_13], properties {weights [1,1,1]}}
 {id=Random_3, children [WavePlayer_0,WavePlayer_1,WavePlayer_2,WavePlayer_5,WavePlayer_6], properties {weights [1,1,1,1,1]}}
 {
  id=WavePlayer_0
  properties {soundWaveAssetPtr=&/PodPlant_Stink_Whip_Highend_01}
 }
 {
  id=WavePlayer_1
  properties {soundWaveAssetPtr=&/PodPlant_Stink_Whip_Highend_02}
 }
 {
  id=WavePlayer_13
  properties {soundWaveAssetPtr=&/PodPlant_Stink_Whip_04}
 }
 {
  id=WavePlayer_2
  properties {soundWaveAssetPtr=&/PodPlant_Stink_Whip_Highend_03}
 }
 {
  id=WavePlayer_5
  properties {soundWaveAssetPtr=&/PodPlant_Stink_Whip_Highend_04}
 }
 {
  id=WavePlayer_6
  properties {soundWaveAssetPtr=&/PodPlant_Stink_Whip_Highend_05}
 }
 {
  id=WavePlayer_7
  properties {soundWaveAssetPtr=&/PodPlant_Stink_Whip_01}
 }
 {
  id=WavePlayer_8
  properties {soundWaveAssetPtr=&/PodPlant_Stink_Whip_02}
 }
]
properties {
 attenuationSettings=/Game/Sounds/Attenuation/Fort_Weapon_Projectile_InAir_Loop_Att
 bDebug=true
 bOverrideAttenuation=true
 maxDistance=1512
 soundClassObject=&Fort_WeaponRanged_Projectile_Effects
 volumeMultiplier=1
 attenuationOverrides {
  bEnableReverbSend=false
  bUseComplexCollisionForOcclusion=true
  falloffDistance=1000
  occlusionVolumeAttenuation=0
  stereoSpread=0
  attenuationShapeExtents {x=512}
 }
}
